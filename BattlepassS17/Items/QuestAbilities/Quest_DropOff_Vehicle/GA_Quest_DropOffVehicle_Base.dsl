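; Base server-side gameplay ability for a vehicle drop-off quest, activated by gameplay event.
; On activation it checks the occupied vehicle's required tags and that the player lacks start-location tags, then binds vehicle exit.
; Exiting with the hardcoded Holly Hedges POI tag sends the quest stat event and plays victory sound and effects.
; The destination-tag wait task has no handler; exited and re-entered events are empty stubs.
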(event Ability|EventActivateAbilityFromEvent
  (params (EventData :type Gameplay%20Event%20Data%20Structure%20%28by%20ref%29))
  ;[1 1 1] Activate
  (|CallParentFunction EventData)
  (if (Networking|IsServer)
    ;[1 1 1] Ensure player is in the correct vehicle and start location
    (bind _vehicle_occupied (Utilities|IsinVehicle (Variables|GAVars|Data|GetOwningPlayerPawn))
      (:true
        (Variables|Default|SetQuestVehicle _vehicle_occupied)
        (if (GameplayTags|HasAllTags (Class|FortVehicleInterface|GetGameplayTags _vehicle_occupied) (Variables|ReadOnly|GetRequiredVehicleTags))
          (if (not (GameplayTags|HasAnyTags (GameplayTags|GetOwnedGameplayTags (Variables|GAVars|Data|GetOwningPlayerPawn) :!class /Script/GameplayTags.GameplayTagAssetInterface) (Variables|ReadOnly|GetStartLocationTags)))
            ;[1 1 1] Listen for destination tag
            (bind _asynctaskproxy (Ability|Tasks|WaitGameplayTagAdd (Variables|ReadOnly|GetDestinationLocationTag) 0 true))
            (Variables|Default|SetTagListenTask _asynctaskproxy)
            ;[1 1 1] Listen for exit vehicle
            (Vehicle|BindEventtoOnExitVehicle (Fortnite|GetFortPlayerControllerfromActor (Variables|GAVars|Data|GetOwningPlayerPawn)) (delegate Custom|ExitVehicle))))
        (Ability|EndAbility)))))

(event Custom|ExitVehicle
  ;[1 1 1] Listen for exit vehicle
  (Variables|Default|Setreentered false)
  (if (GameplayTags|HasTag (GameplayTags|GetOwnedGameplayTags (Variables|GAVars|Data|GetOwningPlayerPawn) :!class /Script/GameplayTags.GameplayTagAssetInterface) "(TagName=\"Athena.Location.POI.HollyHedges\")")
    (Quest|SendComplexCustomStatEvent (Variables|GAVars|Data|GetOwningPlayerQuestManager) 0 (Variables|ReadOnly|GetQuestStatEventTags) "(GameplayTags=)" 1)
    (CallFunction|VictoryFX)))

(event Custom|VictoryFX
  ;[1 1 1] Victory FX
  (Audio|PlaySound2D (Variables|ReadOnly|GetVictorySound))
  (Effects|SpawnFXSystemAttached (Variables|ReadOnly|GetQuestSuccessEmitter) (Variables|/Script/Engine.Character|GetMesh (Variables|GAVars|Data|GetOwningPlayerPawn)) "Feet"))

(event Custom|ExitedVehicle)

(event Custom|Re-EnteredVehicle)
